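# Translucent unlit Niagara sprite and ribbon material for the PlayAsVillain pickup beam, with a scalar that blends between pillar and flare shapes.
# A procedural UV pillar mask and rotating soft-flare texture feed a two-color emissive blend and an opacity using particle alpha, depth fade and camera-distance fade.
# A world position offset branch scales geometry up with camera distance.
# An animated radial-ray pattern appears unconnected.
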Material /PlayAsVillain/Assets/FX/M_PlayAsVillain_PickupBeam
graph {
 properties {
  bAutomaticallySetUsageInEditor=false
  bUsedWithNiagaraRibbons=true
  bUsedWithNiagaraSprites=true
  blendMode=BLEND_Translucent
  shadingModel=MSM_Unlit
  twoSided=true
  usedShadingModels=Unlit
 }
 nodes [
  n1 = ParticleColor()
  n2 = Saturate(None=LinearInterpolate(A=NamedRerouteUsage(), B=NamedRerouteUsage(), Alpha=ScalarParameter() {parameterName=Pillar or Flare}))
  LinearInterpolate(A=Multiply(A=Multiply(A=@n1.A, B=@n2), B=DepthFade()), Alpha=NamedRerouteUsage()) {constA=1, constB=0}
  MaterialFunctionCall(Emissive=LinearInterpolate(
    A=VectorParameter() {parameterName=Color A, defaultValue=<rgba 1 0 0 1>}.RGB
    B=VectorParameter() {parameterName=Color B, defaultValue=<rgba 0 1 0 1>}.RGB
    Alpha=@n2
   )) {materialFunction=&MF_FX_EmissiveAdjustments}
  n3 = TextureCoordinate()
  n4 = MaterialFunctionCall(Vector or UVs=ConstantBiasScale(None=@n3) {bias=-0.5, scale=2}) {materialFunction=&VectorToRadialValue}
  n5 = Time()
  n6 = Panner(Coordinate=@n4.Radial Coordinates, Time=@n5) {speedX=0.1}
  n7 = OneMinus(None=ComponentMask(None=@n6) {g=true})
  n8 = Power(Base=@n7) {constExponent=4}
  n9 = ComponentMask(None=@n6) {r=true}
  n10 = Constant() {r=0.1}
  n11 = OneMinus(None=ComponentMask(None=Panner(Coordinate=@n4.Radial Coordinates, Time=@n5) {speedX=-0.1}) {r=true})
  n12 = Constant() {r=0.1}
  n13 = Multiply(A=Multiply(A=Fmod(A=@n9, B=@n10), B=Fmod(A=@n11, B=@n12))) {constB=100}
  n14 = Power(Base=@n7)
  n15 = Constant() {r=0.025}
  n16 = OneMinus(None=@n9)
  n17 = Constant() {r=0.025}
  n18 = Multiply(A=Multiply(A=Fmod(A=@n9, B=@n15), B=Fmod(A=@n16, B=@n17))) {constB=1000}
  Multiply(A=Multiply(A=Multiply(A=@n8, B=@n13), B=Multiply(A=@n14, B=@n18))) {constB=25}
  n19 = Multiply(A=ScalarParameter() {defaultValue=2, parameterName=MaxScaleIncrease}) {constB=-1}
  n20 = ActorPositionWS()
  n21 = Subtract(A=@n20, B=WorldPosition().XYZ)
  n22 = CameraPositionWS()
  n23 = Divide(
   A=Subtract(A=Distance(A=@n20, B=@n22), B=ScalarParameter() {defaultValue=2000, parameterName=DistanceOffset})
   B=ScalarParameter() {defaultValue=7000, parameterName=ScaleUpDistance}
  )
  NamedRerouteDeclaration(None=LinearInterpolate(B=Multiply(A=@n19, B=@n21) {constB=-1}, Alpha=Clamp(None=@n23)) {constB=0}) {name=to WPO, nodeColor=<rgba 0 1 0.4844 1>}
  NamedRerouteDeclaration(None=TextureSample(UVs=Rotator(Coordinate=@n3) {speed=0.35}) {samplerSource=SSM_Clamp_WorldGroupSettings, texture=/Game/Effects/Fort_Effects/Textures/Flares/T_Flare_Soft}.RGB) {name=Flare Mask, nodeColor=<rgba 1 0.1641 0 1>}
  n24 = TextureCoordinate()
  n25 = ComponentMask(None=@n24) {r=true}
  n26 = OneMinus(None=@n25)
  n27 = ComponentMask(None=@n24) {g=true}
  n28 = Power(Base=Multiply(A=Multiply(A=@n25, B=@n26), B=Power(Base=@n27) {constExponent=1.5}))
  n29 = OneMinus(None=@n27)
  NamedRerouteDeclaration(None=Multiply(A=Multiply(A=@n28) {constB=10}, B=Power(Base=@n29) {constExponent=0.5})) {name=Pillar Mask, nodeColor=<rgba 0 0.1562 1 1>}
  n30 = Length(None=Subtract(A=CameraPositionWS(), B=WorldPosition().XYZ)) {is=class MaterialExpressionLength}
  n31 = ScalarParameter() {defaultValue=1000, group=Camera Fade, parameterName=Camera Fade Far Start, sliderMax=50000}
  n32 = ScalarParameter() {defaultValue=3000, group=Camera Fade, parameterName=Camera Fade Far End, sliderMax=70000}
  n33 = Constant() {r=1}
  n34 = Constant()
  NamedRerouteDeclaration(None=Saturate(None=MaterialFunctionCall(Input=@n30, Input Low=@n31, Input High=@n32, Target Low=@n33, Target High=@n34) {materialFunction=&RemapValueRange}.Result)) {name=Camera Fade Far, nodeColor=<rgba 1 0.0469 0 1>}
  NamedRerouteUsage()
  Panner(Coordinate=@n4.Radial Coordinates, Time=@n5) {speedX=-0.1}
 ]
}
