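# Translucent, unlit, two-sided material for a wind cone effect on Niagara mesh particles in PlayAsVillain.
# Two panning smoke samples, with tiling and speed set by dynamic parameters, form a "Flames" noise for streaks and normal-based vertex offset.
# UV gradients, glow, particle and vertex color shape tint and opacity, softened by view-angle, depth and optional mask or near-camera fades.
# Emissive passes through HDR remapping and time-of-day scaling, with simpler mobile paths.
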Material /PlayAsVillain/Assets/FX/M_PlayAsVillain_WindCone
graph {
 properties {
  bUsedWithNanite=true
  bUsedWithNiagaraMeshParticles=true
  bUsedWithParticleSprites=true
  blendMode=BLEND_Translucent
  pixelDepthOffsetMode=PDOM_Legacy
  shadingModel=MSM_Unlit
  twoSided=true
  usedShadingModels=Unlit
 }
 nodes [
  n1 = DotProduct(A=CameraVectorWS(), B=VertexNormalWS())
  n2 = ScalarParameter() {defaultValue=5, parameterName=Fresnel Size}
  n3 = DynamicParameter() {desc=Niagara index 1, parameterIndex=1, defaultValue=<rgba 0 1 1 1>, paramNames [WPOPinch,GlowMult,Param3,Param4]}
  n4 = ComponentMask(None=TextureCoordinate()) {g=true}
  n5 = ScalarParameter() {defaultValue=6, parameterName=GlowPower}
  n6 = Multiply(
   A=@n3.GlowMult
   B=Multiply(A=ScalarParameter() {defaultValue=0.5, parameterName=GlowBrightness}, B=Power(Base=@n4, Exp=@n5) {constExponent=6}) {constB=0.5}
  )
  n7 = Power(Base=@n4, Exp=ScalarParameter() {defaultValue=1, parameterName=NoiseFadePower})
  n8 = Add(A=@n6, B=NamedRerouteUsage())
  n9 = Multiply(A=ParticleColor().A, B=Add(A=@n6, B=Multiply(A=@n7, B=@n8)))
  n10 = VertexColor()
  n11 = Power(Base=@n10.R) {constExponent=1}
  n12 = DotProduct(A=CameraVectorWS(), B=VertexNormalWS())
  n13 = ScalarParameter() {defaultValue=1.25, parameterName=DirectionalFadePowerLow}
  n14 = ScalarParameter() {defaultValue=4, parameterName=DirectionalFadePower}
  n15 = VertexColor()
  n16 = Multiply(A=Power(Base=Abs(None=@n12), Exp=LinearInterpolate(A=@n13, B=@n14, Alpha=@n15.G) {constA=1.25}) {constExponent=4}) {constB=4}
  n17 = ScalarParameter() {defaultValue=50, parameterName=DepthFadeDist}
  n18 = Multiply(
   A=Multiply(A=Multiply(A=@n9, B=@n11), B=Clamp(None=@n16))
   B=ShadingPathSwitch(Default=DepthFade(FadeDistance=@n17), Mobile=Constant() {r=1}) {desc=Migrated from FeatureLevelSwitch}
  )
  LinearInterpolate(A=@n18, B=Multiply(A=Power(Base=@n1, Exp=@n2), B=@n18), Alpha=TwoSidedSign())
  n19 = ScalarParameter() {defaultValue=0.45, parameterName=HDRMult_3}
  n20 = ScalarParameter() {defaultValue=1, parameterName=HDRMult_2}
  n21 = ScalarParameter() {defaultValue=0.5, parameterName=HDRMult_1}
  n22 = ScalarParameter() {defaultValue=0.35, parameterName=HDRMult}
  n23 = VectorParameter() {parameterName=DarkColor, defaultValue=<rgba 1.5 0.4451 0 1>}
  n24 = VectorParameter() {parameterName=LightColor, defaultValue=<rgba 1 1 1 1>}
  n25 = Multiply(A=@n7, B=@n7)
  n26 = Multiply(
   A=LinearInterpolate(
    A=LinearInterpolate(A=@n23.RGB, B=@n24.RGB, Alpha=@n25)
    B=VectorParameter() {parameterName=CenterColor, defaultValue=<rgba 0 0 0 1>}.RGB
    Alpha=OneMinus(None=@n10.R)
   )
   B=ParticleColor().RGB
  )
  n27 = Clamp(None=Multiply(A=@n11, B=@n26))
  n28 = MaterialFunctionCall(LDR Input=@n27, HDR Tint=Multiply(A=@n26, B=ScalarParameter() {defaultValue=0.35, parameterName=HDRMult})) {materialFunction=&DeriveHDRfromLDR}
  MaterialFunctionCall(Emissive=Multiply(
    A=MaterialFunctionCall(Color_Morning=@n19, Color_Day=@n20, Color_Evening=@n21, Color_Night=@n22) {materialFunction=&Color_Time_of_Day}.Result
    B=ShadingPathSwitch(Default=@n28.RemappedRange, Mobile=@n27) {desc=Migrated from FeatureLevelSwitch}
   )) {materialFunction=&MF_FX_EmissiveAdjustments}
  n29 = MaterialFunctionCall(Float2=TextureCoordinate()) {materialFunction=&BreakOutFloat2Components}
  n30 = OneMinus(None=@n29.R)
  n31 = Multiply(
   A=Power(Base=Multiply(A=@n29.R, B=@n30), Exp=ScalarParameter() {defaultValue=1, parameterName=Directional Mask Power}) {constExponent=1.5}
   B=ScalarParameter() {defaultValue=1, parameterName=Directional Mask Hardness}
  )
  n32 = Fresnel() {baseReflectFraction=0, exponent=1}
  n33 = Multiply(
   A=Power(Base=OneMinus(None=@n32), Exp=ScalarParameter() {defaultValue=2, parameterName=FresnelPower})
   B=ScalarParameter() {defaultValue=3, parameterName=FresnelHardness}
  )
  n34 = StaticSwitchParameter(True=Multiply(A=Saturate(None=@n33), B=@n18), False=@n18) {defaultValue=true, parameterName=Fresnel}
  n35 = StaticSwitchParameter(True=Saturate(None=Multiply(A=@n31, B=@n34)), False=@n34) {group=DetailedGradientMask, parameterName=UseDIrectionalGradientMask}
  n36 = MaterialFunctionCall(Distance=ScalarParameter() {parameterName=CameraFadeDistance}) {materialFunction=&Near_Camera_Fade}
  Multiply(
   A=ParticleColor().A
   B=StaticSwitchParameter(True=Multiply(A=@n35, B=@n36.Result), False=@n35) {group=DetailedGradientMask, parameterName=UseCameraNearFade}
  )
  n37 = TextureCoordinate()
  n38 = VertexNormalWS()
  n39 = NamedRerouteUsage()
  Multiply(
   A=OneMinus(None=ComponentMask(None=@n37) {g=true})
   B=Multiply(A=Multiply(A=@n38, B=@n39), B=ScalarParameter() {defaultValue=100, parameterName=WPO_Intensity}) {constB=100}
  )
  n40 = DynamicParameter() {paramNames [YTile,PanSpeed,Param3,Param4]}
  n41 = AppendVector(A=ScalarParameter() {defaultValue=1, parameterName=X_Tile}, B=@n40.YTile)
  n42 = Multiply(A=@n41, B=TextureCoordinate() {uTiling=3, vTiling=0.5})
  n43 = Multiply(A=Time(), B=@n40.PanSpeed)
  n44 = AppendVector(A=ScalarParameter() {defaultValue=0.2, parameterName=NoiseA_SpeedX}, B=ScalarParameter() {defaultValue=1, parameterName=NoiseA_SpeedY})
  n45 = Multiply(A=@n41, B=TextureCoordinate() {uTiling=3, vTiling=0.2})
  n46 = AppendVector(A=ScalarParameter() {defaultValue=-0.25, parameterName=NoiseB_SpeedX}, B=ScalarParameter() {defaultValue=0.8, parameterName=NoiseB_SpeedY})
  n47 = LinearInterpolate(
   A=TextureSampleParameter2D(UVs=Panner(Coordinate=@n42, Time=@n43, Speed=@n44) {speedX=0.2, speedY=1}) {parameterName=NoiseA, texture=&T_Inky_Smoke_Tile_Chromatic}.G
   B=TextureSampleParameter2D(UVs=Panner(Coordinate=@n45, Time=@n43, Speed=@n46) {speedX=-0.25, speedY=0.8}) {parameterName=NoiseB, texture=&T_Inky_Smoke_Tile_Chromatic}.R
  )
  n48 = ScalarParameter() {defaultValue=1, parameterName=Noise_Exponent}
  NamedRerouteDeclaration(None=Multiply(A=Power(Base=@n47, Exp=@n48)) {constB=1.5}) {name=Flames, nodeColor=<rgba 1 0.0703 0 1>}
  n49 = ComponentMask(None=Panner(Coordinate=@n37) {speedY=2.3}) {g=true}
  n50 = Power(Base=Sine(None=Multiply(A=@n49) {constB=0.5})) {constExponent=1.5}
  Saturate(None=Add(A=@n39, B=Multiply(A=@n50)))
 ]
}
