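# Niagara system for the PlayAsVillain smoke crown effect; as configured, it renders an orange body aura and flickering zaps on character sockets.
# NE_BodyAura sprites and NE_Zaps meshes cycle socket indices by system age, add random offsets, and scale with an invisibility alpha.
# The crown itself is disabled: a dark ribbon arc with orange follower ribbons reading it through ParticleRead, trailing owner velocity and suppressed for the local player.
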NiagaraSystem /PlayAsVillain/Assets/FX/NS_PlayAsVillain_SmokeCrown
dependencies {
 usedDataInterfaces [
  &NiagaraDataInterfaceCurve
  &NiagaraDataInterfaceParticleRead
  &NiagaraDataInterfaceArrayFloat3
  &NiagaraDataInterfaceColorCurve
  &NiagaraDataInterfaceArrayFloat
  &NiagaraDataInterfaceSocketReader
 ]
 usedDynamicInputs [&Boolean_NotOperation, /Game/Effects/Niagara/DynamicInputs/State/FN_IsExecutionStateActive]
 usedModules [
  &self:SystemSpawnScript.SystemScriptSource.SystemScriptGraph.NiagaraNodeAssignment_0.SetVariables_#1
  &self:SystemSpawnScript.SystemScriptSource.SystemScriptGraph.NiagaraNodeAssignment_1.SetVariables_#2
  &self:SystemSpawnScript.SystemScriptSource.SystemScriptGraph.NiagaraNodeAssignment_4.SetVariables_#3
  &SystemState
  &self:SystemSpawnScript.SystemScriptSource.SystemScriptGraph.NiagaraNodeAssignment_2.SetVariables_#4
  &Timeline
  &self:SystemSpawnScript.SystemScriptSource.SystemScriptGraph.NiagaraNodeAssignment_3.SetVariables_#5
  &self:SystemSpawnScript.SystemScriptSource.SystemScriptGraph.NiagaraNodeAssignment_5.SetVariables_#6
 ]
}
emitters {
 NE_BodyAura {
  emitterUpdateScript [
   &EmitterState()
   &SpawnRate(SpawnRate=&Multiply_Float(A=16, B=$System.InvisibilityAlpha), Use Spawn Probability=false)
  ]
  particleSpawnScript [
   /SetVariables(Particles.SocketIndex:NiagaraInt32=&MakeIntFromFloat(
     Float=&Multiply_Float(
      A=&Lerp_Float(A=0, B=11, Alpha=&ModuloFloat(A=&Multiply_Float(A=$Engine.System.Age, B=1), B=1))
      B=1
     )
     Conversion Mode=::NewEnumerator1
    )) {isSetParameters=true}
   &InitializeParticle(
    Lifetime=2.5
    Sprite Rotation Mode=::NewEnumerator2
    Sprite Size=&Multiply_Vector2D_ByFloat(
     Float=&Multiply_Float(
      A=35
      B=&SelectFloatFromArray(
       Direct Array Index=$Particles.SocketIndex
       Float Selection Array=DataInterface() {propertyValues {FloatData [1,2,2,2,1.5,1,1.5,1,2,1.5,2,1.5]}}
      )
     )
     Vector2D=&Vector2DFromFloat(Value=1)
    )
    Sprite Size Mode=::NewEnumerator4
    UsePositionOffset=false
    Write Material Random=true
   )
   &CameraOffset(Camera Offset Amount=-10)
  ]
  particleUpdateScript [
   &ParticleState()
   &SocketLocation(
    SamplePosOffset=true
    Sampled Position Offset=&Multiply_VectorByFloat(
     Vector=&SelectVectorFromArray(
      Direct Array Index=$Particles.SocketIndex
      Vector Selection Array=DataInterface() {propertyValues {FloatData [<0 0 0>, <20 0 0>, <20 0 0>, <20 0 0>, <30 0 0>, <20 0 0>, <-30 0 0>, <-20 0 0>, <-40 0 0>, <-30 0 0>, <40 0 0>, <30 0 0>]}}
     )
     Float=&RandomRangeFloat(Minimum=0, Maximum=1)
    )
    Socket Index=$Particles.SocketIndex
    Socket Reader=$System.Socket Reader
    Socket Sampling Mode=::NewEnumerator3
    VelocitySpeedLimit=false
   )
   &DynamicMaterialParameters(
    Index 0 Param 1=&RandomRangeFloat(Minimum=0, Maximum=1)
    Module.Index0Float0Write=true
    Module.Index0Float1Write=true
    Module.Index0Float2Write=true
    Module.Index0Float3Write=true
   )
   &Color(Color=&MultiplyLinearColor(
     Linear Color=&ColorFromCurve(ColorCurve=DataInterface() {propertyValues {
        AlphaCurve {keys [<at 0 0>,<at 0.2025 0.02 RCIM_Cubic>,<at 1 0 RCIM_Cubic>]}
        BlueCurve {keys [<at 0 0.1355 RCIM_Cubic>,<at 0.2152 0.342>,<at 1 0.1086 RCIM_Cubic>]}
        GreenCurve {keys [<at 0 0.5652 RCIM_Cubic>,<at 0.2152 1.0479>,<at 1 0.7863 RCIM_Cubic>]}
        RedCurve {keys [<at 0 1 RCIM_Cubic>,<at 0.2152 1.5>,<at 1 1 RCIM_Cubic>]}
       }})
     Scale Factor=&MakeVector4FromVectorAndFloat("Vector (RGB)"=<1 1 1>, "Float (Alpha)"=$System.InvisibilityAlpha)
    ))
   &FN_Sprite_Attributes(
    Color=&MultiplyLinearColor(
     Linear Color=&ColorFromCurve(ColorCurve=DataInterface() {propertyValues {
        AlphaCurve {keys [<at 0 0>,<at 0.2025 0.1 RCIM_Cubic>,<at 1 0 RCIM_Cubic>]}
        BlueCurve {keys [<at 0 0.1355 RCIM_Cubic>,<at 0.2152 0.1328>,<at 1 0.1086 RCIM_Cubic>]}
        GreenCurve {keys [<at 0 0.5652 RCIM_Cubic>,<at 0.2152 0.833>,<at 1 0.7863 RCIM_Cubic>]}
        RedCurve {keys [<at 0 1 RCIM_Cubic>,<at 0.2152 1.5>,<at 1 1 RCIM_Cubic>]}
       }})
     Scale Factor=&MakeVector4FromVectorAndFloat("Vector (RGB)"=<1 1 1>, "Float (Alpha)"=$System.InvisibilityAlpha)
    )
    Create Color Attribute=true
    Create SpriteSize Attribute=true
    SpriteSize=&Multiply_Vector2D_ByFloat(Float=0.2, Vector2D=$Particles.SpriteSize)
    Write to Module Namespace=true
   )
   /SetVariables(Particles.SmokeSubUv:NiagaraFloat=&RandomRangeFloat(Minimum=0, Maximum=3)) {isSetParameters=true}
  ]
  rendererClasses [&NiagaraSpriteRendererProperties]
  renderers [
   {rendererClass=&NiagaraSpriteRendererProperties}
   {rendererClass=&NiagaraSpriteRendererProperties, rendererIndex=1}
  ]
 }
 NE_CrownRibbon {
  bEnabled=false
  emitterUpdateScript [
   &EmitterState(Life Cycle Mode=::NewEnumerator1, Loop Behavior=::NewEnumerator1, Loop Duration Mode=::NewEnumerator1, UseLoopDelay=false)
   &SpawnBurst_Instantaneous(
    Spawn Count=&MakeCustomIntFromBool(Bool=$User.IsLocal, True INT=0, False INT=$System.CrownParticles)
    Use Spawn Probability=false
    UseLoopCountLimit=false
   )
  ]
  particleSpawnScript [&InitializeParticle(
    Color=&MakeCustomLinearColorFromBool(Bool=$User.IsLocal, True Color=<rgba 0.005 0.005 0.005 0.05>, False Color=<rgba 0.01 0.01 0.01 0.15>)
    Lifetime=5
    Ribbon Width=&MakeCustomFloatFromBool(Bool=$User.IsLocal, True Float=3, False Float=7)
    Ribbon Width Mode=::NewEnumerator1
    Sprite Rotation Mode=::NewEnumerator2
    Sprite Size Mode=::NewEnumerator3
    Uniform Sprite Size=15
    UsePositionOffset=false
    Write Material Random=true
   )]
  particleUpdateScript [
   &ParticleState(Kill Particles When Lifetime Has Elapsed=false)
   &SocketLocation(
    Position Sampling=::NewEnumerator1
    SamplePosOffset=true
    Sampled Position Offset=$System.CrownOffset
    Socket Reader=$System.Socket Reader
    Socket Sampling Mode=::NewEnumerator3
    VelocitySpeedLimit=false
   )
   &RotateAroundPoint(
    Radius:NiagaraFloat=&Multiply_Float(
     A=&Lerp_Float(
      A=13
      B=45
      Alpha=&FloatFromCurve(
       FloatCurve=DataInterface() {propertyValues {Curve {keys [
           {arriveTangent=3.3961, interpMode=RCIM_Cubic, leaveTangent=3.3961, tangentMode=RCTM_User, time=0, value=0}
           {interpMode=RCIM_Cubic, tangentMode=RCTM_User, time=0.5, value=1}
           {arriveTangent=-2.7206, interpMode=RCIM_Cubic, leaveTangent=-2.7206, tangentMode=RCTM_User, time=1, value=0}
          ]}}}
       CurveIndex=&ReturnNormalizedExecIndex()
      )
     )
     B=$System.CrownScale
    )
    Rotation Center:NiagaraPosition=$Output.SocketLocation.SampledPosition
    Rotation Phase:NiagaraFloat=&Lerp_Float(A=-0.25, B=0.25, Alpha=&ReturnNormalizedExecIndex())
    Set Particle Position=false
    X Axis:Vector3f=&RotateVectorByQuaternion(Vector To Rotate=<1 -0.5 0>, Quaternion=$Output.SocketLocation.SampledRotation)
    Y Axis:Vector3f=&RotateVectorByQuaternion(Vector To Rotate=<0 0 1>, Quaternion=$Output.SocketLocation.SampledRotation)
   )
   /SetVariables(Particles.GoalPos:NiagaraPosition=$Output.RotateAroundPoint.OutputPosition) {isSetParameters=true}
   /SetVariables(Particles.Position=$Particles.GoalPos) {isSetParameters=true}
   /SetVariables(Particles.Position=$Output.RotateAroundPoint.OutputPosition) {enabled=false, isSetParameters=true}
   /SetVariables(Particles.Position=&AddVectorToPosition(
     Position=$Particles.Position
     Vector=&Multiply_Vector(
      A=&Multiply_VectorByFloat(
       Vector=&NormalizeVector(
        Vector To Normalize=&TransformVector(Vector=$Engine.Owner.Velocity, Source Space=::World, Destination Space=::Local)
        Safe Normalize=true
        Safe Fallback Threshold=0
        Safe Fallback Vector=<0 0 0.001>
       )
       Float=&ClampFloat(
        Float=&VectorLength(Vector=&TransformVector(Vector=$Engine.Owner.Velocity, Source Space=::World, Destination Space=::Local))
        Max=1500
       )
      )
      B=&Multiply_VectorByFloat(
       Vector=<-1 -1 -0.1>
       Float=&FloatFromCurve(
        FloatCurve=DataInterface() {propertyValues {Curve {keys [
            {arriveTangent=0.09, interpMode=RCIM_Cubic, leaveTangent=0.09, tangentMode=RCTM_User, time=0.1, value=0}
            <at 0.5 1 RCIM_Cubic>
            {arriveTangent=0.1423, interpMode=RCIM_Cubic, leaveTangent=0.1423, tangentMode=RCTM_User, time=0.9, value=0}
           ]}}}
        CurveIndex=&ReturnNormalizedExecIndex()
        Scale Curve=0.005
       )
      )
     )
    )) {enabled=false, isSetParameters=true}
   &ScaleRibbonWidth(Ribbon Width Scale=&Lerp_Float(
     A=0.1
     B=3
     Alpha=&FloatFromCurve(
      FloatCurve=DataInterface() {propertyValues {Curve {keys [
          {arriveTangent=3.3961, interpMode=RCIM_Cubic, leaveTangent=3.3961, tangentMode=RCTM_User, time=0, value=0}
          {interpMode=RCIM_Cubic, tangentMode=RCTM_User, time=0.5, value=1}
          {arriveTangent=-2.7206, interpMode=RCIM_Cubic, leaveTangent=-2.7206, tangentMode=RCTM_User, time=1, value=0}
         ]}}}
      CurveIndex=&ReturnNormalizedExecIndex()
     )
    ))
   &FN_Sprite_Attributes(
    Color=&MultiplyLinearColor(
     Linear Color=$Particles.Color
     Scale Factor=&MakeVector4FromVectorAndFloat("Vector (RGB)"=<1 1 1>, "Float (Alpha)"=&Multiply_Float(A=$System.FadeIn, B=$System.FadeOut))
    )
    Create Color Attribute=true
    Create SpriteSize Attribute=true
    SpriteSize=&Multiply_Vector2D_ByFloat(
     Float=&Lerp_Float(
      A=0
      B=30
      Alpha=&FloatFromCurve(
       FloatCurve=DataInterface() {propertyValues {Curve {keys [
           {arriveTangent=4.6208, interpMode=RCIM_Cubic, leaveTangent=4.6208, tangentMode=RCTM_User, time=0, value=0}
           {arriveTangent=2, interpMode=RCIM_Cubic, leaveTangent=2, time=0.1, value=0.6}
           {interpMode=RCIM_Cubic, tangentMode=RCTM_User, time=0.5, value=1}
           {arriveTangent=-2, interpMode=RCIM_Cubic, leaveTangent=-2, time=0.9, value=0.6}
           {arriveTangent=-4.852, interpMode=RCIM_Cubic, leaveTangent=-4.852, tangentMode=RCTM_User, time=1, value=0}
          ]}}}
       CurveIndex=&ReturnNormalizedExecIndex()
      )
     )
     Vector2D=<1 1>
    )
    Write to Module Namespace=true
   )
   &DynamicMaterialParameters(
    Index 3 Param 1=&RandomRangeFloat(Minimum=0, Maximum=1)
    Module.Index0Float0Write=true
    Module.Index0Float1Write=true
    Module.Index0Float2Write=true
    Module.Index0Float3Write=true
    Module.Index3Float0Write=true
    Module.Index3Float1Write=true
    Module.Index3Float2Write=true
    Module.Index3Float3Write=true
    Write Parameter Index 3=::NewEnumerator1
   )
   /SetVariables(
    Particles.SecondRibbonColor:LinearColor=&MultiplyLinearColor(
     Linear Color=$Particles.Color
     Scale Factor=&MakeVector4FromVectorAndFloat("Vector (RGB)"=<1 1 1>, "Float (Alpha)"=1.5)
    )
    Particles.SecondRibbonLinkOrder:NiagaraFloat=&ReturnNormalizedExecIndex()
    Particles.SecondRibbonUVOOveride:NiagaraFloat=&ReturnNormalizedExecIndex()
    Particles.SecondRibbonWidth:NiagaraFloat=&Multiply_Float(A=$Particles.RibbonWidth, B=1.2)
   ) {isSetParameters=true, module=SetVariables_#7}
   &ScaleColor(
    Scale Alpha=&Multiply_Float(A=&Multiply_Float(A=$System.FadeIn, B=$System.FadeOut), B=1)
    ScaleA=true
    ScaleRGB=true
   )
   &KillParticles(Kill Particles=&SetBoolByFloatComparison(A=$System.FadeOut, B=0.05, Comparison Type=::NewEnumerator6))
  ]
  rendererClasses [&NiagaraSpriteRendererProperties, &NiagaraRibbonRendererProperties]
  renderers [
   {rendererClass=&NiagaraSpriteRendererProperties}
   {rendererClass=&NiagaraRibbonRendererProperties, rendererIndex=1}
   {rendererClass=&NiagaraRibbonRendererProperties, rendererIndex=2}
  ]
 }
 NE_RedRibbons {
  bEnabled=false
  emitterSpawnScript [/SetVariables(Emitter.ParticlePerRibbon=10, Emitter.RibbonNumber=4) {isSetParameters=true}]
  emitterUpdateScript [
   &EmitterState(Life Cycle Mode=::NewEnumerator1, Loop Behavior=::NewEnumerator1, Loop Duration Mode=::NewEnumerator1, UseLoopDelay=false)
   &SpawnBurst_Instantaneous(
    Spawn Count=&MakeCustomIntFromBool(
     Bool=$User.IsLocal
     True INT=0
     False INT=&Multiply_Int(A=$Emitter.RibbonNumber, B=$Emitter.ParticlePerRibbon)
    )
    Use Spawn Probability=false
    UseLoopCountLimit=false
   )
  ]
  particleSpawnScript [
   &InitializeParticle(
    Color=<rgba 1 0.449 0.0648 1>
    Lifetime=5
    Ribbon Width=&MakeCustomFloatFromBool(Bool=$User.IsLocal, True Float=0.5, False Float=1.5)
    Ribbon Width Mode=::NewEnumerator1
    Sprite Size Mode=::NewEnumerator3
    Uniform Sprite Size=5
    UsePositionOffset=false
    Write Material Random=true
   )
   &PartitionParticles(
    ID to Partition:NiagaraInt32=&ReturnExecIndex()
    Normalized Index In Partition Includes One=true
    Number Of Partitions=$Emitter.RibbonNumber
    Particles Per Partition:NiagaraInt32=$Emitter.ParticlePerRibbon
    Partition Mode:ENiagara_PartitionMode_v2=::NewEnumerator1
    Write Partition Index to Ribbon Link Order:ENiagara_AttributeSamplingApplyOutput=::NewEnumerator0
    Write Partition to Ribbon ID=::NewEnumerator0
   )
   /SetVariables(Particles.NormalizedPartition:NiagaraFloat=&Lerp_Float(
     A=0
     B=1
     Alpha=&FloatFromCurve(
      FloatCurve=DataInterface() {propertyValues {Curve {keys [<at 0 0>,<at 1 1>]}}}
      CurveIndex=&Divide_Float(
       A=&MakeFloatFromInt(Int=$Particles.PartitionParticles.Partition)
       B=&MakeFloatFromInt(Int=/Niagara/DynamicInputs/Subtract/Subtract_Integer(A:NiagaraInt32=$Emitter.RibbonNumber, B:NiagaraInt32=0))
      )
     )
    )) {enabled=false, isSetParameters=true}
   /SetVariables(Particles.NormalizedPartition:NiagaraFloat=&FloatFromCurve(
     FloatCurve=DataInterface() {propertyValues {Curve {keys [<at 0 0.15>,<at 1 1>]}}}
     CurveIndex=$Particles.PartitionParticles.NormalizedPartition
    )) {isSetParameters=true}
  ]
  particleUpdateScript [
   &ParticleState(Kill Particles When Lifetime Has Elapsed=false)
   &SocketLocation(
    Orientation Sampling=::NewEnumerator1
    Position Sampling=::NewEnumerator1
    SamplePosOffset=false
    Socket Reader=$System.Socket Reader
    Socket Sampling Mode=::NewEnumerator3
    VelocitySpeedLimit=false
   )
   &self:FollowRibbon(
    Alpha:NiagaraFloat=$Particles.NormalizedPartition
    ParticlePerRibbon:NiagaraInt32=$System.CrownParticles
    RibbonReader=DataInterface() {propertyValues {EmitterBinding {bindingMode=Other, emitterName=NE_CrownRibbon}}}
   ) {module=FollowRibbon}
   &RotateAroundPoint(
    Radius:NiagaraFloat=&Lerp_Float(
     A=25
     B=50
     Alpha=&FloatFromCurve(
      FloatCurve=DataInterface() {propertyValues {Curve {keys [
          {arriveTangent=3.3961, interpMode=RCIM_Cubic, leaveTangent=3.3961, tangentMode=RCTM_User, time=0, value=0}
          {interpMode=RCIM_Cubic, tangentMode=RCTM_User, time=0.5, value=1}
          {arriveTangent=-2.7206, interpMode=RCIM_Cubic, leaveTangent=-2.7206, tangentMode=RCTM_User, time=1, value=0}
         ]}}}
      CurveIndex=$Particles.NormalizedPartition
     )
    )
    Rotation Center:NiagaraPosition=$Output.SocketLocation.SampledPosition
    Rotation Phase:NiagaraFloat=&Lerp_Float(A=-0.25, B=0.25, Alpha=$Particles.NormalizedPartition)
    Set Particle Position=false
    X Axis:Vector3f=&RotateVectorByQuaternion(Vector To Rotate=<1 -0.5 0>, Quaternion=$Output.SocketLocation.SampledRotation)
    Y Axis:Vector3f=<1 0 0>
   )
   /SetVariables(Particles.GoalPos:NiagaraPosition=&AddVectorToPosition(
     Position=&TransformPosition(Position=$Output.FollowRibbon.RibbonPos, Source Space=::World, Destination Space=::Local)
     Vector=&Multiply_VectorByFloat(
      Vector=&Lerp_Vector(
       A=<0 0 -5>
       B=&Lerp_Vector(
        A=<0 0 40>
        B=<0 0 50>
        Alpha=&FloatFromCurve(
         FloatCurve=DataInterface() {propertyValues {Curve {keys [
             {arriveTangent=3.3961, interpMode=RCIM_Cubic, leaveTangent=3.3961, tangentMode=RCTM_User, time=0, value=0}
             {interpMode=RCIM_Cubic, tangentMode=RCTM_User, time=0.5, value=1}
             {arriveTangent=-2.7206, interpMode=RCIM_Cubic, leaveTangent=-2.7206, tangentMode=RCTM_User, time=1, value=0}
            ]}}}
         CurveIndex=$Particles.NormalizedPartition
        )
       )
       Alpha=$Particles.PartitionParticles.Normalized Index In Partition
      )
      Float=&Multiply_Float(A=$System.FadeIn, B=$System.FadeOut)
     )
    )) {isSetParameters=true}
   /SetVariables(Particles.Position=/Niagara/DynamicInputs/TypeConversions/MakeCustomPositionFromBool(
     Bool:NiagaraBool=$Output.ParticleState.FirstFrame
     True Position:NiagaraPosition=$Particles.GoalPos
     False Position:NiagaraPosition=&Lerp_Vector(
      A=Linked() {linkedVariable {name=Particles.GoalPos, type=NiagaraPosition}}
      B=&Lerp_Vector(
       A=Linked() {linkedVariable {name=Particles.Position, type=NiagaraPosition}}
       B=Linked() {linkedVariable {name=Particles.GoalPos, type=NiagaraPosition}}
       Alpha=0.1
      )
      Alpha=&FloatFromCurve(
       FloatCurve=DataInterface() {propertyValues {Curve {keys [
           {arriveTangent=3.5003, interpMode=RCIM_Cubic, leaveTangent=3.5003, tangentMode=RCTM_User, time=0.2, value=0}
           {interpMode=RCIM_Cubic, tangentMode=RCTM_User, time=1, value=1}
          ]}}}
       CurveIndex=$Particles.PartitionParticles.Normalized Index In Partition
      )
     )
    )) {isSetParameters=true}
   /SetVariables(Particles.Position=&AddVectorToPosition(
     Position=$Particles.Position
     Vector=&Multiply_Vector(
      A=&Multiply_VectorByFloat(
       Vector=&NormalizeVector(
        Vector To Normalize=&TransformVector(Vector=$Engine.Owner.Velocity, Source Space=::World, Destination Space=::Local)
        Safe Normalize=true
        Safe Fallback Threshold=0
        Safe Fallback Vector=<0 0 0.001>
       )
       Float=&ClampFloat(
        Float=&VectorLength(Vector=&TransformVector(Vector=$Engine.Owner.Velocity, Source Space=::World, Destination Space=::Local))
        Max=1500
       )
      )
      B=&Multiply_VectorByFloat(
       Vector=<-1 -1 -1>
       Float=&Lerp_Float(
        A=0
        B=0.005
        Alpha=&FloatFromCurve(
         FloatCurve=DataInterface() {propertyValues {Curve {keys [<at 0.2 0>,<at 1 1>]}}}
         CurveIndex=$Particles.PartitionParticles.Normalized Index In Partition
        )
       )
      )
     )
    )) {isSetParameters=true}
   /SetVariables(Particles.Position=&AddVectorToPosition(
     Position=$Particles.Position
     Vector=&Multiply_Vector(
      A=&Multiply_VectorByFloat(
       Vector=&NormalizeVector(Vector To Normalize=$Engine.Owner.Velocity, Safe Normalize=true, Safe Fallback Threshold=0, Safe Fallback Vector=<0 0 0.001>)
       Float=&ClampFloat(Float=&VectorLength(Vector=$Engine.Owner.Velocity), Max=1500)
      )
      B=&Multiply_VectorByFloat(
       Vector=<1 1 1>
       Float=&Lerp_Float(
        A=0
        B=0.05
        Alpha=&FloatFromCurve(
         FloatCurve=DataInterface() {propertyValues {Curve {keys [<at 0.25 0>,<at 1 1>]}}}
         CurveIndex=$Particles.PartitionParticles.Normalized Index In Partition
        )
       )
      )
     )
    )) {enabled=false, isSetParameters=true}
   &ScaleColor(
    Scale Alpha=&Multiply_Float(
     A=&Multiply_Float(
      A=&FloatFromCurve(
       FloatCurve=DataInterface() {propertyValues {Curve {keys [
           {arriveTangent=-2.8351, interpMode=RCIM_Cubic, leaveTangent=-2.8351, tangentMode=RCTM_User, time=0, value=1}
           {arriveTangent=-0.1884, interpMode=RCIM_Cubic, leaveTangent=-0.1884, tangentMode=RCTM_User, time=0.3, value=0.2}
           {interpMode=RCIM_Cubic, tangentMode=RCTM_User, time=1, value=0.2}
          ]}}}
       CurveIndex=$Particles.PartitionParticles.Normalized Index In Partition
      )
      B=&FloatFromCurve(FloatCurve=DataInterface() {propertyValues {Curve {keys [<at 0 0>,<at 0.1 1>]}}}, CurveIndex=$Particles.NormalizedPartition)
     )
     B=&MakeCustomFloatFromBool(Bool=$User.IsLocal, True Float=0.5, False Float=1)
    )
    ScaleA=true
    ScaleRGB=true
   )
   &ScaleRibbonWidth(Ribbon Width Scale=&Lerp_Float(
     A=1
     B=2
     Alpha=&FloatFromCurve(
      FloatCurve=DataInterface() {propertyValues {Curve {keys [
          {arriveTangent=3.3961, interpMode=RCIM_Cubic, leaveTangent=3.3961, tangentMode=RCTM_User, time=0, value=0}
          {interpMode=RCIM_Cubic, tangentMode=RCTM_User, time=0.5, value=1}
          {arriveTangent=-4.3839, interpMode=RCIM_Cubic, leaveTangent=-4.3839, tangentMode=RCTM_User, time=0.8, value=0}
         ]}}}
      CurveIndex=$Particles.PartitionParticles.NormalizedPartition
     )
    ))
   &FN_Sprite_Attributes(
    Color=&MultiplyLinearColor(
     Linear Color=<rgba 1 0.3222 0 0.15>
     Scale Factor=&MakeVector4FromVectorAndFloat("Vector (RGB)"=<1 1 1>, "Float (Alpha)"=&Multiply_Float(A=$System.FadeIn, B=$System.FadeOut))
    )
    Create Color Attribute=true
    Create SpriteSize Attribute=true
    SpriteSize=&Multiply_Vector2D_ByFloat(
     Float=&Multiply_Float(
      A=4
      B=&FloatFromCurve(
       FloatCurve=DataInterface() {propertyValues {Curve {keys [<at 0 0 RCIM_Constant>,<at 0.1 1 RCIM_Constant>,<at 0.2 0 RCIM_Constant>]}}}
       CurveIndex=$Particles.PartitionParticles.Normalized Index In Partition
      )
     )
     Vector2D=<2 3>
    )
    Write to Module Namespace=true
   )
   &DynamicMaterialParameters(
    Index 0 Param 1=/Game/Effects/Niagara/DynamicInputs/Random/FN_SeededRandomRangeFloat(Minimum:NiagaraFloat=0, Maximum:NiagaraFloat=1, Random Seed:NiagaraInt32=$Particles.PartitionParticles.Partition)
    Index 3 Param 1=&RandomRangeFloat(Minimum=0, Maximum=1)
    Module.Index0Float0Write=true
    Module.Index0Float1Write=true
    Module.Index0Float2Write=true
    Module.Index0Float3Write=true
    Module.Index3Float0Write=true
    Module.Index3Float1Write=true
    Module.Index3Float2Write=true
    Module.Index3Float3Write=true
    Write Parameter Index 3=::NewEnumerator1
   )
   /SetVariables(
    Particles.SecondRibbonColor:LinearColor=&MultiplyLinearColor(Linear Color=$Particles.Color, Scale Factor=&MakeVector4_FromLinearColor(LinearColor=<rgba 1 1 1 0.25>))
    Particles.SecondRibbonLinkOrder:NiagaraFloat=$Particles.PartitionParticles.Normalized Index In Partition
    Particles.SecondRibbonUVOOveride:NiagaraFloat=$Particles.PartitionParticles.Normalized Index In Partition
    Particles.SecondRibbonWidth:NiagaraFloat=&Multiply_Float(A=$Particles.RibbonWidth, B=1.5)
   ) {isSetParameters=true, module=SetVariables_#7}
   &SolveForcesAndVelocity(Clamp Velocity=false, Limit Acceleration=false)
   &KillParticles(Kill Particles=&SetBoolByFloatComparison(A=$System.FadeOut, B=0.05, Comparison Type=::NewEnumerator6))
  ]
  rendererClasses [&NiagaraSpriteRendererProperties, &NiagaraRibbonRendererProperties]
  renderers [
   {rendererClass=&NiagaraSpriteRendererProperties}
   {rendererClass=&NiagaraRibbonRendererProperties, rendererIndex=1}
   {rendererClass=&NiagaraRibbonRendererProperties, rendererIndex=2}
  ]
 }
 NE_Zaps {
  emitterUpdateScript [
   &EmitterState()
   &SpawnRate(
    Spawn Probability=0.8
    SpawnRate=&Multiply_Float(A=6, B=$System.InvisibilityAlpha)
    Use Spawn Probability=true
   )
  ]
  particleSpawnScript [
   /SetVariables(Particles.SocketIndex:NiagaraInt32=&MakeIntFromFloat(
     Float=&Multiply_Float(
      A=&Lerp_Float(
       A=0
       B=11
       Alpha=&ModuloFloat(A=&Multiply_Float(A=$Engine.System.Age, B=3.352), B=1)
      )
      B=1
     )
     Conversion Mode=::NewEnumerator1
    )) {isSetParameters=true}
   &InitializeParticle(
    Lifetime=&RandomRangeFloat(Minimum=0.45, Maximum=0.55)
    Mesh Scale Mode=::NewEnumerator3
    Mesh Uniform Scale=&RandomRangeFloat(Minimum=0.1, Maximum=0.12)
    Sprite Size Mode=::NewEnumerator3
    Uniform Sprite Size=20
    UsePositionOffset=false
    Write Material Random=true
   )
   &InitialMeshOrientation(
    Mesh Orientation Mode=::NewEnumerator2
    Orientation Vector=&MaskVectorBySpawnGroup(Masked Vector=&RandomVector(Vector Scale=1), Passthrough Vector=<1 0 0>)
    Rotation=&MaskVectorBySpawnGroup(Masked Vector=<0 0 0>, Passthrough Vector=&RandomRangeVector(Minimum=<0 0 0>, Maximum=<0 0 1>))
    Use Rotation Vector=true
   )
   &CameraOffset(Camera Offset Amount=10)
  ]
  particleUpdateScript [
   &ParticleState()
   &SocketLocation(
    SamplePosOffset=true
    Sampled Position Offset=&Multiply_VectorByFloat(
     Vector=&SelectVectorFromArray(
      Direct Array Index=$Particles.SocketIndex
      Vector Selection Array=DataInterface() {propertyValues {FloatData [
         <0 0 0>
         <0 0 0>
         <0 0 0>
         <0 0 0>
         <0 0 0>
         <10 0 0>
         <-10 0 0>
         <10 0 0>
         <-10 0 0>
         <-20 0 0>
         <20 0 0>
         <-20 0 0>
         <20 0 0>
         <0 0 0>
         <0 0 0>
        ]}}
     )
     Float=&RandomRangeFloat(Minimum=0, Maximum=1)
    )
    Socket Index=$Particles.SocketIndex
    Socket Reader=$System.Socket Reader
    Socket Sampling Mode=::NewEnumerator3
    VelocitySpeedLimit=false
   )
   &ScaleMeshSize(Scale Factor=&Multiply_VectorByFloat(
     Vector=&RandomRangeVector(Minimum=<1 1 2.2>, Maximum=<1.3 1.3 2.2>)
     Float=&Multiply_Float(
      A=&FloatFromCurve(FloatCurve=DataInterface() {propertyValues {Curve {keys [<at 0 1>,<at 1 1.5>]}}})
      B=&Lerp_Float(A=1, B=2, Alpha=$System.BuildUp)
     )
    )) {module=ScaleMeshSize001}
   &DynamicMaterialParameters(
    Index 0 Param 2=&RandomRangeFloat(Minimum=0, Maximum=0)
    Index 0 Param 4=&MaskFloatBySpawnGroup(Masked Float=13, Passthrough Float=1)
    Index 1 Param 1=&FloatFromCurve(FloatCurve=DataInterface() {propertyValues {Curve {keys [<at 0 0.5>,<at 1 1>]}}})
    Index 1 Param 2=2
    Index 1 Param 3=&FloatFromCurve(FloatCurve=DataInterface() {propertyValues {Curve {keys [{interpMode=RCIM_Cubic, tangentMode=RCTM_Break, time=0.5, value=-0.5}, <at 1 0.1>]}}})
    Index 2 Param 1=&Multiply_Float(
     A=&FloatFromCurve(
      FloatCurve=DataInterface() {propertyValues {Curve {keys [{interpMode=RCIM_Cubic, leaveTangent=8.1085, tangentMode=RCTM_Break, time=0.8568, value=1}, {tangentMode=RCTM_User, time=1, value=0}]}}}
      Scale Curve=2
     )
     B=&Divide_Float(
      A=&MakeFloatFromVector(Vector=$Particles.Scale, Channel=::NewEnumerator0)
      B=&RandomRangeFloat(Minimum=10, Maximum=15)
     )
    )
    Index 2 Param 2=&FloatFromCurve(
     FloatCurve=DataInterface() {propertyValues {Curve {keys [
         {interpMode=RCIM_Cubic, tangentMode=RCTM_Break, time=0.7, value=0}
         {arriveTangent=-0.0859, interpMode=RCIM_Cubic, tangentMode=RCTM_Break, time=1, value=-1}
        ]}}}
     Scale Curve=50
    )
    Index 2 Param 3=&Waveform(
     Global Input Position=&RandomRangeFloat(Minimum=0, Maximum=15)
     Global Phase=$Particles.Age
     "Waveform [1]"=::NewEnumerator12
     "[1] Phase Scale"=2
     "[1] Frequency"=1
     "[1] Amplitude Min/Max"=<0 1>
    )
    Index 2 Param 4=25
    Index 3 Param 1=&FloatFromCurve(
     FloatCurve=DataInterface() {propertyValues {Curve {keys [
         {interpMode=RCIM_Cubic, leaveTangent=-0.0025, tangentMode=RCTM_Break, time=0, value=0}
         {arriveTangent=1.8528, interpMode=RCIM_Cubic, tangentMode=RCTM_Break, time=1, value=1}
        ]}}}
     Scale Curve=&RandomRangeFloat(Minimum=0, Maximum=0)
    )
    Index 3 Param 2=&FloatFromCurve(
     FloatCurve=DataInterface() {propertyValues {Curve {keys [
         {interpMode=RCIM_Cubic, leaveTangent=-0.0025, tangentMode=RCTM_Break, time=0, value=0.5}
         {arriveTangent=-0.0316, interpMode=RCIM_Cubic, tangentMode=RCTM_Break, time=1, value=-0.2}
        ]}}}
     Scale Curve=0.1
    )
    Index 3 Param 3=&FloatFromCurve(FloatCurve=DataInterface() {propertyValues {Curve {keys [
         {interpMode=RCIM_Cubic, leaveTangent=1.1034, tangentMode=RCTM_Break, time=0, value=-1}
         {arriveTangent=2.2399, tangentMode=RCTM_User, time=0.7, value=0.125}
        ]}}})
    Index 3 Param 4=&FloatFromCurve(FloatCurve=DataInterface() {propertyValues {Curve {keys [
         {interpMode=RCIM_Cubic, leaveTangent=3.0601, tangentMode=RCTM_Break, time=0.5, value=0}
         {arriveTangent=3.2053, tangentMode=RCTM_User, time=1, value=1}
        ]}}})
    Module.Index0Param0Write=true
    Module.Index0Param1Write=true
    Module.Index0Param2Write=true
    Module.Index0Param3Write=true
    Module.Index1Param0Write=true
    Module.Index1Param1Write=true
    Module.Index1Param2Write=true
    Module.Index1Param3Write=true
    Module.Index2Param0Write=true
    Module.Index2Param1Write=true
    Module.Index2Param2Write=true
    Module.Index2Param3Write=true
    Module.Index3Param0Write=true
    Module.Index3Param1Write=true
    Module.Index3Param2Write=true
    Module.Index3Param3Write=true
    Write Parameter Index 1=true
    Write Parameter Index 2=true
    Write Parameter Index 3=true
   )
   &SolveForcesAndVelocity(Clamp Velocity=false, Limit Acceleration=false)
   &Color(
    Color=&MultiplyLinearColor(
     Linear Color=&ColorFromCurve(ColorCurve=DataInterface() {propertyValues {
        AlphaCurve {keys [<at 0 1 RCIM_Cubic>,<at 0.7884 1>,<at 1 0 RCIM_Cubic>]}
        BlueCurve {keys [<at 0.1421 1.0189>, {interpMode=RCIM_Cubic, leaveTangent=-19.7604, time=0.4058, value=0}]}
        GreenCurve {keys [<at 0.1421 3.8387>, {interpMode=RCIM_Cubic, leaveTangent=-19.7604, time=0.4058, value=0.4072}]}
        RedCurve {keys [<at 0.1421 10>, {interpMode=RCIM_Cubic, leaveTangent=-19.7604, time=0.4058, value=1}]}
       }})
     Scale Factor=&MakeVector4FromVectorAndFloat("Vector (RGB)"=<1 1 1>, "Float (Alpha)"=$System.InvisibilityAlpha)
    )
    Scale Alpha=&FloatFromCurve(FloatCurve=DataInterface() {propertyValues {Curve {keys [
         {interpMode=RCIM_Cubic, leaveTangent=2.5294, tangentMode=RCTM_Break, time=0, value=0}
         {arriveTangent=-0.2166, interpMode=RCIM_Cubic, tangentMode=RCTM_Break, time=0.4, value=1}
        ]}}})
   )
   /SetVariables(Particles.DecalColor:LinearColor=&MultiplyLinearColor(Linear Color=<rgba 0.2 0.2 0.2 1>, Scale Factor=&MakeVector4_FromLinearColor(LinearColor=$Particles.Color))) {isSetParameters=true}
   /SetVariables(Particles.GlowDecalSize:Vector2f=&Multiply_Vector2D_ByFloat(Float=60, Vector2D=<1 1>)) {isSetParameters=true}
   /SetVariables(Particles.DecalDynMat:Vector4f=<4 1 0 0>) {isSetParameters=true}
  ]
  rendererClasses [&NiagaraMeshRendererProperties, &NiagaraSpriteRendererProperties]
  renderers [
   {rendererClass=&NiagaraMeshRendererProperties, rendererIndex=0}
   {rendererClass=&NiagaraMeshRendererProperties, rendererIndex=1}
   {rendererClass=&NiagaraSpriteRendererProperties, rendererIndex=2}
  ]
 }
}
summary {
 systemName=NS_PlayAsVillain_SmokeCrown
 emitters [
  {
   bEnabled=false
   emitterName=NE_CrownRibbon
   rendererClasses [&NiagaraSpriteRendererProperties, &NiagaraRibbonRendererProperties]
  }
  {
   bEnabled=false
   emitterName=NE_RedRibbons
   rendererClasses [&NiagaraSpriteRendererProperties, &NiagaraRibbonRendererProperties]
  }
  {emitterName=NE_Zaps, rendererClasses [&NiagaraMeshRendererProperties, &NiagaraSpriteRendererProperties]}
  {emitterName=NE_BodyAura, rendererClasses [&NiagaraSpriteRendererProperties]}
 ]
 userVariables [
  User.BossIndex {type=NiagaraInt32}
  User.Invisibility {type=NiagaraFloat, defaultValue=0, description=""}
  User.IsLocal {type=NiagaraBool}
 ]
}
