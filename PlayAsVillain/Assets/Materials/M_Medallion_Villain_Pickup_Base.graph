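# Base material for a villain medallion pickup that bobs, spins and shows a glinting light sweep.
# It inserts effects between the character parent PreFX and PostFX functions, using material attributes and named reroutes.
# World position offset combines optional coin bobble with Z-axis rotation, while vertex-interpolated rotated normals blend into the surface normal.
# A panning band along a camera-space direction, weighted by Fresnel and tinted, multiplies base color into emissive.
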Material /PlayAsVillain/Assets/Materials/M_Medallion_Villain_Pickup_Base
graph {
 properties {
  bAutomaticallySetUsageInEditor=false
  bUseMaterialAttributes=true
  pixelDepthOffsetMode=PDOM_Legacy
  shadingModel=MSM_FromMaterialExpression
  usedShadingModels=Default Lit | Subsurface | Clear Coat | Subsurface Profile | Cloth
 }
 nodes [
  n1 = MaterialFunctionCall() {materialFunction=&MF_CharacterParent_2023_PreFX}
  MaterialFunctionCall(Material=NamedRerouteUsage(), M=@n1.M, Pre Skinned Local Position=@n1.M, Distance Blend=@n1.M, Fresnel=@n1.M) {materialFunction=&MF_CharacterParent_2023_PostFX}
  n2 = Constant()
  n3 = MaterialFunctionCall!(TimeScale=ScalarParameter() {defaultValue=0.4, group=Medallion, parameterName=BobbleSpeed}, MaxHeight, MinHeight) {materialFunction=/Game/Playgrounds/Items/Props/Materials/MF_Creative_CoinBobble}
  n4 = ScalarParameter() {defaultValue=1, group=Medallion, parameterName=EnableBobble}
  n5 = Constant3Vector() {constant=<rgba 0 0 1 0>}
  n6 = Multiply(A=Time(), B=ScalarParameter() {defaultValue=0.1, group=Medallion, parameterName=RotationSpeed, sortPriority=2})
  n7 = ObjectPositionWS()
  n8 = WorldPosition()
  NamedRerouteDeclaration(None=Add(
    A=LinearInterpolate(A=@n2, B=@n3.WorldPositionOffset, Alpha=@n4)
    B=RotateAboutAxis(NormalizedRotationAxis=@n5, RotationAngle=@n6, PivotPoint=@n7, Position=@n8.XYZ)
   )) {name=WPO, nodeColor=<rgba 0 0.8828 1 1>}
  NamedRerouteDeclaration(None=VertexInterpolator(VS=MaterialFunctionCall(Rotation Axis=@n5, Rotation Angle=@n6) {materialFunction=&FixRotateAboutAxisNormals}.New Tangent Space Vertex Normal).PS) {name=RotatedTangentNormals, nodeColor=<rgba 1 0.7969 0 1>}
  NamedRerouteDeclaration(None=@n1.Fresnel) {name=Fresnel, nodeColor=<rgba 0.4766 0 1 1>}
  NamedRerouteDeclaration(None=@n1.Result) {name=MA_In, nodeColor=<rgba 0 1 0.1562 1>}
  n9 = GetMaterialAttributes(None=NamedRerouteUsage()) {attributeGetTypes [#1, #2]}
  n10 = Add(A=@n9.Emissive Color, B=Multiply(A=@n9.Base Color, B=NamedRerouteUsage()))
  n11 = Saturate(None=ScalarParameter() {defaultValue=1, group=Medallion, parameterName=UseSweep})
  n12 = GetMaterialAttributes(None=SetMaterialAttributes(MaterialAttributes=@n9.MaterialAttributes, Emissive Color=LinearInterpolate(A=@n9.Emissive Color, B=@n10, Alpha=@n11)) {attributeSetTypes [#1]}) {attributeGetTypes [#3, #4]}
  n13 = NamedRerouteUsage()
  n14 = NamedRerouteUsage()
  NamedRerouteDeclaration(None=SetMaterialAttributes!(
    MaterialAttributes=@n12.MaterialAttributes
    World Position Offset=Add(A=@n12.World Position Offset, B=@n13)
    Normal=MaterialFunctionCall(BaseNormal=@n14, AdditionalNormal=@n12.Normal) {materialFunction=&BlendAngleCorrectedNormals}.Result
   ) {attributeSetTypes [#3, #4]}) {name=MA_Out, nodeColor=<rgba 1 0 0.1641 1>}
  n15 = Multiply(
   A=Subtract(A=WorldPosition().XYZ, B=ObjectPositionWS())
   B=ScalarParameter() {defaultValue=0.01, group=Medallion, parameterName=SweepWorldTileScale}
  )
  n16 = Normalize(VectorInput=Transform(None=VectorParameter() {group=Medallion, parameterName=SweepDirectionCamSpace, defaultValue=<rgba -0.5 0.45 -1 0>}.RGB) {transformSourceType=TRANSFORMSOURCE_Camera})
  n17 = ComponentMask(None=Multiply(A=DotProduct(A=@n15, B=@n16), B=@n16)) {b=true, r=true}
  n18 = ScalarParameter() {defaultValue=0.7, group=Medallion, parameterName=SweepSpeed}
  n19 = Frac(None=ComponentMask(None=Panner(Coordinate=@n17, Speed=@n18) {speedX=0.2, speedY=0.2}) {g=true})
  n20 = Multiply(A=Subtract(A=@n19) {constB=0.9}) {constB=10}
  n21 = Multiply(A=Floor(None=Add(A=@n19) {constB=0.1}), B=Power(Base=OneMinus(None=@n20)) {constExponent=3})
  n22 = Saturate(None=Add(A=NamedRerouteUsage()) {constB=0.1})
  NamedRerouteDeclaration(None=Multiply(A=Multiply(A=@n21, B=@n22), B=VectorParameter() {group=Medallion, parameterName=SweepColor, defaultValue=<rgba 20 14.6562 2.4956 1>}.RGB)) {name=Sweep, nodeColor=<rgba 0.3828 1 0 1>}
 ]
}
